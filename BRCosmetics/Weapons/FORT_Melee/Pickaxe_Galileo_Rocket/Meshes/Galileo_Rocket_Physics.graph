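# Physics asset for the Galileo Rocket two-handed pickaxe cosmetic, defining collision bodies and joint limits for its dangling straps.
# A long box covers the handle, which anchors a four-box strap chain linked bone to bone with tight, uniform swing and twist limits.
# Separate top and bottom strap spheres, offset from their bones, constrain directly to the handle with wider, asymmetric swing limits.
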PhysicsAsset /BRCosmetics/Weapons/FORT_Melee/Pickaxe_Galileo_Rocket/Meshes/Galileo_Rocket_Physics
bodies {
 Melee_TwoHanded_Handle [{shapeType=Box, extentX=4.2151, extentY=4.2151, extentZ=148.8527, center=<0 -1.4921 0>}]
 dyn_strap_01 [{shapeType=Box, extentX=1.01, extentY=14.2299, extentZ=25.6496, center=<5.7992 0 -0.0004>, rotation=<rot 90 0 -90>}]
 dyn_strap_02 [{shapeType=Box, extentX=1.1877, extentY=11.8192, extentZ=25.7576, center=<5.1296 0 0.0001>, rotation=<rot 90 -26.565 -116.565>}]
 dyn_strap_03 [{shapeType=Box, extentX=5.1955, extentY=9.7752, extentZ=26.598, center=<5.2734 0 0.0001>, rotation=<rot 90 -26.565 -116.565>}]
 dyn_strap_04 [{shapeType=Box, extentX=1.01, extentY=10.3419, extentZ=22.311, center=<4.5741 0 0>, rotation=<rot 90 0 -90>}]
 dyn_strap_bottom [{shapeType=Sphere, radius=10, center=<31.681 -0.0003 -28.6972>}]
 dyn_strap_top [{shapeType=Sphere, radius=10, center=<28.4592 0.0002 34.2145>}]
}
constraints [
 {bone1Name=dyn_strap_top, bone2Name=Melee_TwoHanded_Handle, swing1LimitDegrees=37, swing2LimitDegrees=65, twistLimitDegrees=10}
 {bone1Name=dyn_strap_bottom, bone2Name=Melee_TwoHanded_Handle, swing1LimitDegrees=37, swing2LimitDegrees=65, twistLimitDegrees=10}
 {bone1Name=dyn_strap_01, bone2Name=Melee_TwoHanded_Handle, swing1LimitDegrees=22, swing2LimitDegrees=22, twistLimitDegrees=8}
 {bone1Name=dyn_strap_02, bone2Name=dyn_strap_01, swing1LimitDegrees=22, swing2LimitDegrees=22, twistLimitDegrees=8}
 {bone1Name=dyn_strap_03, bone2Name=dyn_strap_02, swing1LimitDegrees=22, swing2LimitDegrees=22, twistLimitDegrees=8}
 {bone1Name=dyn_strap_04, bone2Name=dyn_strap_03, swing1LimitDegrees=22, swing2LimitDegrees=22, twistLimitDegrees=8}
]
